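# MetaSound Source for the long steam burst one-shot SFX of the BasilOkra Meltdown set.
# A single stereo Wave Player fires the pipe-steam burst wave on play and signals OnFinished when playback ends.
# The source uses a one-shot attenuation asset, the in-game SFX sound class, a capped concurrency override, and volume modulation through the BasilOkra control bus.
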MetaSoundSource /WildEstate_BasilOkra/Assets/Audio/SFX/MSS_BasilOkra_Meltdown_SteamBurst_Long
inputs [
 Steam Burst Long {type=WaveAsset}
 UE.Source.OnPlay {type=Trigger}
]
interfaces [UE.OutputFormat.Stereo:1.0, UE.Source.OneShot:1.0, UE.Source:1.1]
outputs [
 "UE.OutputFormat.Stereo.Audio:0" {type=Audio}
 "UE.OutputFormat.Stereo.Audio:1" {type=Audio}
 UE.Source.OneShot.OnFinished {type=Trigger}
]
pages [Default {
  input_defaults {Steam Burst Long=/WildEstate_BasilOkra/Assets/Audio/SFX/wav/BasilOkra_Meltdown_PipeSteam_Burst_OS, UE.Source.OnPlay=null}
  nodes [WavePlayer_0 = UE.Wave Player:Stereo(Wave Asset=^Steam Burst Long, Play=^UE.Source.OnPlay) {also {version [1,1]}}]
  links [
   {dst="@UE.Source.OneShot.OnFinished", src=WavePlayer_0.On Finished}
   {dst="@UE.OutputFormat.Stereo.Audio:0", src=WavePlayer_0.Out Left}
   {dst="@UE.OutputFormat.Stereo.Audio:1", src=WavePlayer_0.Out Right}
  ]
 }]
properties {
 assetImportData=&self:AssetImportData
 attenuationSettings=&/ATT_BasilOkra_Meltdown_OS
 bOverrideConcurrency=true
 maxDistance=6000
 numChannels=2
 outputFormat=Stereo
 sampleRate=48000
 soundClassObject=&Fort_SFX_InGame
 volume=0.3
 concurrencyOverrides {bVolumeScaleCanRelease=true, maxCount=5, voiceStealReleaseTime=0.2, volumeScaleReleaseTime=0.2}
 modulationSettings {volumeRouting=Override, volumeModulationDestination {bEnableModulation=true, modulators [/WildEstate_BasilOkra/Assets/Audio/CB_BasilOkra]}}
}
