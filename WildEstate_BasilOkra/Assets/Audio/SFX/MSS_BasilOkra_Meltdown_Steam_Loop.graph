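# Mono looping MetaSound source for the small steam leak ambience of the BasilOkra meltdown in Wild Estate.
# On play it randomly picks a gas leak or steam valve wave, triggering an ADSR envelope whose attack starts the wave player.
# The envelope and a float Multiplier input scale the player's output for a gradual fade-in.
# A small-steam attenuation asset and a control bus volume modulator shape the in-game mix.
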MetaSoundSource /WildEstate_BasilOkra/Assets/Audio/SFX/MSS_BasilOkra_Meltdown_Steam_Loop
inputs [
 Multiplier {type=Float}
 Steam Loop {type=WaveAsset:Array}
 UE.Source.OnPlay {type=Trigger}
]
interfaces [UE.OutputFormat.Mono:1.0, UE.Source:1.1]
outputs ["UE.OutputFormat.Mono.Audio:0" {type=Audio}]
pages [Default {
  input_defaults {
   Multiplier=0.8
   UE.Source.OnPlay=null
   Steam Loop [/WildEstate_BasilOkra/Assets/Audio/SFX/wav/BasilOkra_GasLeak_Loop, /WildEstate_BasilOkra/Assets/Audio/SFX/wav/BasilOkra_SteamValve_Loop]
  }
  nodes [
   ArrayRandomGet_0 = Array.Random Get:WaveAsset:Array(In Array=^Steam Loop, Next=^UE.Source.OnPlay)
   ADSREnvelopeADSREnvelope_0 = ADSR Envelope.ADSR Envelope:Audio(Trigger Attack=@ArrayRandomGet_0.On Next) {also {defaults {Attack Time=1, Decay Time=20}}}
   WavePlayer_0 = UE.Wave Player:Mono(Play=@ADSREnvelopeADSREnvelope_0.On Attack Triggered, Wave Asset=@ArrayRandomGet_0.Value) {also {defaults {Loop="<MetaSoundBuilderResult.SUCCEEDED: 0>"}, version [1,1]}}
   Multiply_1 = UE.Multiply:Audio by Float(
    AdditionalOperands=^Multiplier
    PrimaryOperand=UE.Multiply:Audio(AdditionalOperands=UE.Reroute(Value=@ADSREnvelopeADSREnvelope_0.Out Envelope).Value, PrimaryOperand=@WavePlayer_0.Out Mono).Out
   )
  ]
  links [{dst="@UE.OutputFormat.Mono.Audio:0", src=Multiply_1.Out}]
 }]
properties {
 assetImportData=&self:AssetImportData
 attenuationSettings=&/ATT_BasilOkra_Meltdown_Steam_Small
 bLooping=true
 duration=10000
 maxDistance=1020
 sampleRate=48000
 soundClassObject=&Fort_SFX_InGame
 modulationSettings {volumeRouting=Override, volumeModulationDestination {bEnableModulation=true, modulators [/WildEstate_BasilOkra/Assets/Audio/CB_BasilOkra]}}
}
