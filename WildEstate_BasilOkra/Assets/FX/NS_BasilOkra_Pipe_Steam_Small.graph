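# Niagara system for a small pipe steam jet with burst and looping meltdown steam audio, sized by FXScale and optionally tinted via Is_Purple.
# InitialFlameBurst throws a short sprite puff along the owner's Y axis, while MainBurst builds the ribbon jet and AdditionalFlames cycles SubUV sprites along it.
# User.Reveal drives jet length and fade.
# Jet positions are computed directly, blending owner rotation toward a time-averaged quaternion so the plume bends when turned.
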NiagaraSystem /WildEstate_BasilOkra/Assets/FX/NS_BasilOkra_Pipe_Steam_Small
dependencies {
 usedDataInterfaces [&NiagaraDataInterfaceCurve, &NiagaraDataInterfaceAudioPlayer, &NiagaraDataInterfaceColorCurve]
 usedDynamicInputs [
  &SimulationPosition
  &FNTimeOfDayScalar
  &ClampFloat
 ]
 usedModules [
  &self:SystemSpawnScript.SystemScriptSource.SystemScriptGraph.NiagaraNodeAssignment_0.SetVariables_#1
  &self:SystemSpawnScript.SystemScriptSource.SystemScriptGraph.NiagaraNodeAssignment_3.SetVariables_#2
  &SystemState
  &self:SystemSpawnScript.SystemScriptSource.SystemScriptGraph.NiagaraNodeAssignment_1.SetVariables_#3
  &InterpolateOverTime
  &self:SystemSpawnScript.SystemScriptSource.SystemScriptGraph.NiagaraNodeAssignment_2.SetVariables_#4
 ]
}
emitters {
 AdditionalFlames {
  emitterSpawnScript [
   /SetVariables(
    Emitter.Length=122
    Emitter.ParticlesToSpawn=7
    Emitter.Random:NiagaraFloat=&RandomRangeFloat(Minimum=0, Maximum=1)
   ) {isSetParameters=true, module=SetVariables_#5}
   &InitializeAudioPlayer(Audio Player=DataInterface() {propertyValues {
      Attenuation=/WildEstate_BasilOkra/Assets/Audio/SFX/ATT_BasilOkra_Meltdown_Steam
      Concurrency=/WildEstate_BasilOkra/Assets/Audio/SFX/CONC_BasilOkra_Sparks
      MaxPlaysPerTick=1
      SoundToPlay=/WildEstate_BasilOkra/Assets/Audio/SFX/MSS_BasilOkra_Meltdown_Steam_Loop
      bAllowLoopingOneShotSounds=false
      bLimitPlaysPerTick=true
      bOnlyActiveDuringGameplay=false
      bStopWhenComponentIsDestroyed=true
      ConfigurationUserParameter {parameter {name=None, typeDefHandle {registeredTypeIndex=1}, varData []}}
      ParameterNames []
     }})
  ]
  emitterUpdateScript [
   &EmitterState(Life Cycle Mode=::NewEnumerator1, Loop Duration=1, UseLoopDelay=false)
   &SpawnBurst_Instantaneous(Spawn Count=$Emitter.ParticlesToSpawn, Spawn Probability=0.5, Use Spawn Probability=false, UseLoopCountLimit=false)
   /SetVariables(Emitter.Reveal:NiagaraFloat=&Multiply_Float(
     A=&Lerp_Float(
      A=&FloatFromCurve(FloatCurve=DataInterface() {propertyValues {Curve {keys [<at 0 0>,<at 0.6 1>]}}}, CurveIndex=$User.Reveal)
      B=&FloatFromCurve(FloatCurve=DataInterface() {propertyValues {Curve {keys [<at 0 0>,<at 0.95 1>]}}}, CurveIndex=$User.Reveal)
      Alpha=$System.SystemRandom
     )
     B=&Lerp_Float(A=1, B=1.1, Alpha=$Emitter.Random)
    )) {isSetParameters=true, module=SetVariables_#6}
   /SetVariables(Emitter.Length:NiagaraFloat=&Multiply_Float(
     A=$Emitter.Initial.Length
     B=&FloatFromCurve(
      FloatCurve=DataInterface() {propertyValues {Curve {keys [
          {interpMode=RCIM_Cubic, leaveTangent=6.4155, tangentMode=RCTM_Break, time=0, value=0}
          {arriveTangent=0.1191, interpMode=RCIM_Cubic, leaveTangent=0.1191, tangentMode=RCTM_User, time=0.1, value=0.9}
          {arriveTangent=0.1177, interpMode=RCIM_Cubic, leaveTangent=0.1177, tangentMode=RCTM_User, time=1, value=1}
         ]}}}
      CurveIndex=$Emitter.Reveal
      Scale Curve=&Multiply_Float(A=$User.FXScale, B=&Lerp_Float(A=0.7, B=1, Alpha=$System.SizeRandom))
     )
    )) {isSetParameters=true, module=SetVariables_#7}
   /Niagara/Modules/Audio/PlayLoopingAudio(Update position every frame=false)
  ]
  particleSpawnScript [&InitializeParticle(
    Sprite Rotation Mode=::NewEnumerator2
    Sprite Size Mode=::NewEnumerator4
    Sprite UV Mode=::NewEnumerator1
    UsePositionOffset=false
    Write Material Random=true
   )]
  particleUpdateScript [
   &ParticleState()
   &SpriteRotationRate(Rotation Rate=&Multiply_Float(
     A=&RandomRangeFloat(Minimum=100, Maximum=150)
     B=&UniformAOrBFloat(A=-1, Evaluation Type=::NewEnumerator0, Recalculate Random Each Loop=true, Override Seed=false)
    ))
   /SetVariables(Particles.ParticleRelativeLifetimeSprites:NiagaraFloat=&Frac_Float(Float=&Add_Float(A=&Multiply_Float(A=1.5, B=$Emitter.Age), B=&ReturnNormalizedExecIndex()))) {isSetParameters=true, module=SetVariables_#8}
   /SetVariables(Particles.OpacityMultiplierSprites:NiagaraFloat=&FloatFromCurve(
     FloatCurve=DataInterface() {propertyValues {Curve {keys [
         {interpMode=RCIM_Cubic, tangentMode=RCTM_User, time=0, value=0}
         {arriveTangent=-0.6677, interpMode=RCIM_Cubic, leaveTangent=-0.6677, tangentMode=RCTM_User, time=0.2, value=1}
         {arriveTangent=-0.6675, tangentMode=RCTM_User, time=1.5, value=0}
        ]}}}
     CurveIndex=&Add_Float(
      A=&Multiply_Float(A=$Particles.ParticleRelativeLifetimeSprites, B=-1)
      B=&Multiply_Float(A=$Emitter.Reveal, B=2)
     )
    )) {isSetParameters=true, module=SetVariables_#9}
   /SetVariables(Particles.InterpolateCurveSprites:NiagaraFloat=&FloatFromCurve(
     FloatCurve=DataInterface() {propertyValues {Curve {keys [
         {arriveTangent=-0.6503, interpMode=RCIM_Cubic, leaveTangent=1.4736, tangentMode=RCTM_Break, time=0, value=0}
         {arriveTangent=0.4202, interpMode=RCIM_Cubic, tangentMode=RCTM_Break, time=1, value=1}
        ]}}}
     CurveIndex=$Particles.ParticleRelativeLifetimeSprites
    )) {isSetParameters=true, module=SetVariables_#10}
   /SetVariables(Particles.PositionSprites:NiagaraPosition=&AddVectorToPosition(
     Position=&AddVectorToPosition(
      Position=/Niagara/DynamicInputs/Helpers/SimulationPosition()
      Vector=&TransformVector(Vector=<0 0 0>, Destination Space=::Simulation)
     )
     Vector=&Multiply_VectorByFloat(
      Vector=&RotateVectorByQuaternion(
       Vector To Rotate=<0 1 0>
       Quaternion=&LerpQuaternion(
        Quaternion A:Quat4f=$Engine.Owner.Rotation
        Quaternion B:Quat4f=$System.InterpolateOverTime.InterpolateOverTime_Quaternion.Moving Average
        Lerp Factor:NiagaraFloat=$Particles.InterpolateCurveSprites
        Enable Spherical Linear Interpolation:NiagaraBool=false
       )
      )
      Float=&Multiply_Float(A=$Emitter.Length, B=$Particles.InterpolateCurveSprites)
     )
    )) {isSetParameters=true, module=SetVariables_#11}
   /SetVariables(Particles.SpriteSize=&Multiply_Vector2D(
     A=&Vector2DFromFloat(Value=&FloatFromCurve(
       FloatCurve=DataInterface() {propertyValues {Curve {keys [{arriveTangent=0.3985, time=0, value=0.3}, {arriveTangent=0.3977, leaveTangent=0.3977, time=1, value=1}]}}}
       CurveIndex=$Particles.InterpolateCurveSprites
       Scale Curve=&Lerp_Float(A=70, B=100, Alpha=&RandomRangeFloat(Minimum=0, Maximum=1))
      ))
     B=&Vector2DFromFloat(Value=&Multiply_Float(A=$User.FXScale, B=&Lerp_Float(A=0.7, B=1, Alpha=$System.SizeRandom)))
    )) {isSetParameters=true, module=SetVariables_#12}
   &SubUVAnimation(Start Frame=20, SubUV Lookup Index:NiagaraFloat=$Particles.ParticleRelativeLifetimeSprites)
   &ScaleColor(
    Scale Alpha=&Multiply_Float(
     A=&Multiply_Float(
      A=&FloatFromCurve(
       FloatCurve=DataInterface() {propertyValues {Curve {keys [
           {arriveTangent=16.5612, interpMode=RCIM_Cubic, leaveTangentWeight=0.3333, time=0, value=0}
           {arriveTangentWeight=0.1333, interpMode=RCIM_Cubic, leaveTangentWeight=0.1667, tangentMode=RCTM_User, time=0.333, value=1}
           {arriveTangentWeight=0.1667, interpMode=RCIM_Cubic, tangentMode=RCTM_User, time=1, value=0}
          ]}}}
       CurveIndex=$Particles.ParticleRelativeLifetimeSprites
       Scale Curve=0.9
      )
      B=&Multiply_Float(
       A=&FloatFromCurve(
        FloatCurve=DataInterface() {propertyValues {Curve {keys [
            {interpMode=RCIM_Cubic, tangentMode=RCTM_Break, time=0.6, value=1}
            {arriveTangent=-4.2498, interpMode=RCIM_Cubic, tangentMode=RCTM_Break, time=0.99, value=0}
           ]}}}
        CurveIndex=$Emitter.Reveal
        Scale Curve=&Lerp_Float(A=1, B=1, Alpha=$Emitter.Random)
       )
       B=$System.OpacityRandom
      )
     )
     B=$Particles.OpacityMultiplierSprites
    )
    Scale RGB=&MakeCustomVectorFromBool(Bool=$User.Is_Purple, True Vector=<0.7 0 1>, False Vector=<0.9 0.9 0.9>)
    ScaleA=true
    ScaleRGB=true
   )
   &CameraOffset(Camera Offset Amount=5)
   &ScaleColor(
    Scale Alpha=&FloatFromCurve(
     FloatCurve=DataInterface() {propertyValues {Curve {
        defaultValue=1
        keys [{interpMode=RCIM_Cubic, tangentMode=RCTM_User, time=0, value=0}, {interpMode=RCIM_Cubic, tangentMode=RCTM_User, time=1, value=1}]
       }}}
     CurveIndex=$Emitter.Age
    )
    ScaleA=true
    ScaleRGB=true
   ) {module=ScaleColor001}
  ]
  rendererClasses [&NiagaraSpriteRendererProperties]
  renderers [{rendererClass=&NiagaraSpriteRendererProperties}]
 }
 InitialFlameBurst {
  emitterUpdateScript [
   &EmitterState(Life Cycle Mode=::NewEnumerator1, Loop Behavior=::NewEnumerator1, Loop Duration=1, UseLoopDelay=false)
   &SpawnBurst_Instantaneous(
    Spawn Count=&RandomRangeInt(Minimum=8, Maximum=12)
    Spawn Time=0.03
    Use Spawn Probability=false
    UseLoopCountLimit=false
   )
  ]
  particleSpawnScript [
   &InitializeParticle(
    Color=<rgba 1 1 1 0.6>
    Lifetime=&UniformRangedFloat(Minimum=0.4, Maximum=0.6)
    Mass Mode=::NewEnumerator1
    Sprite Rotation Mode=::NewEnumerator2
    Sprite Size=&Multiply_Vector2D_ByFloat(
     Float=$System.FX_Scale
     Vector2D=&Vector2DFromFloat(Value=&UniformRangedFloat(Minimum=70, Maximum=100))
    )
    Sprite Size Mode=::NewEnumerator4
    UsePositionOffset=false
    Write Material Random=true
   ) {module=InitializeParticle001}
   &ShapeLocation(
    MaskSpawnGroup=false
    Offset Coordinate Space=::Local
    Offset Mode=::NewEnumerator0
    Override Seed=false
    Sphere Radius=0.1
    UseOwnerScale=true
   ) {module=ShapeLocation001}
   &AddVelocity(
    Cone Angle=5
    Cone Axis=$Engine.Owner.SystemYAxis
    Module.UseConeFalloff=false
    Module.UseVelDistribution=false
    Override Seed=false
    Rotation Coordinate Space=::World
    Velocity Mode=::NewEnumerator2
    Velocity Speed=&Multiply_Float(
     A=$System.FX_Scale
     B=&Multiply_Float(
      A=&RandomRangeFloat(Minimum=5, Maximum=1000)
      B=&Lerp_Float(A=0.7, B=1, Alpha=$System.SizeRandom)
     )
    )
   ) {module=AddVelocity001}
  ]
  particleUpdateScript [
   &ParticleState()
   &SubUVAnimation(Start Frame=40, SubUV Lookup Index:NiagaraFloat=&Multiply_Float(A=$Particles.NormalizedAge, B=1))
   &SpriteRotationRate(Rotation Rate=&Multiply_Float(
     A=&RandomRangeFloat(Minimum=100, Maximum=150)
     B=&UniformAOrBFloat(A=-1, Evaluation Type=::NewEnumerator0, Recalculate Random Each Loop=true, Override Seed=false)
    ))
   &InheritSourceMovement(ApplyForce=false, ApplyPosition=true, ApplyPositionFalloff=false)
   &SolveForcesAndVelocity(Clamp Velocity=false, Limit Acceleration=false)
   &ScaleColor(
    Scale Alpha=&Multiply_Float(
     A=&FloatFromCurve(
      FloatCurve=DataInterface() {propertyValues {Curve {keys [
          {arriveTangent=5.5857, interpMode=RCIM_Cubic, leaveTangent=5.5857, leaveTangentWeight=0.3333, tangentMode=RCTM_User, time=0, value=0.3}
          {arriveTangentWeight=0.1333, interpMode=RCIM_Cubic, leaveTangentWeight=0.1667, tangentMode=RCTM_User, time=0.1, value=1}
          {arriveTangentWeight=0.1667, interpMode=RCIM_Cubic, tangentMode=RCTM_User, time=1, value=0}
         ]}}}
      Scale Curve=&Multiply_Float(A=0.3, B=$System.OpacityRandom)
     )
     B=&MakeFloatFromVector(
      Vector=&MakeCustomVectorFromBool(Bool=$User.Is_Purple, True Vector=<0.7 0 1>, False Vector=<0.9 0.9 0.9>)
      Channel=::NewEnumerator0
     )
    )
    Scale RGB=&Multiply_VectorByFloat(
     Vector=&Multiply_VectorByFloat(Vector=<1 1 1>, Float=1)
     Float=&FloatFromCurve(
      FloatCurve=DataInterface() {propertyValues {Curve {keys [
          {arriveTangent=9.0725, interpMode=RCIM_Cubic, leaveTangent=9.0725, tangentMode=RCTM_User, time=0, value=0}
          {arriveTangent=-2.5249, interpMode=RCIM_Cubic, leaveTangent=-2.5249, tangentMode=RCTM_User, time=0.1408, value=2}
          {arriveTangent=-0.0405, interpMode=RCIM_Cubic, leaveTangent=-0.0405, tangentMode=RCTM_User, time=1, value=0}
         ]}}}
      Scale Curve=&RandomRangeFloat(Minimum=0.9, Maximum=1.5)
     )
    )
    ScaleA=true
    ScaleRGB=true
   )
   &ScaleSpriteSize(Scale Factor=&Vector2DFromFloat(Value=&FloatFromCurve(
      FloatCurve=DataInterface() {propertyValues {Curve {keys [<at 0 0.1>,<at 1 1>]}}}
      Scale Curve=&Multiply_Float(A=3, B=&Lerp_Float(A=0.7, B=1, Alpha=$System.SizeRandom))
     ))) {module=ScaleSpriteSize001}
   &CameraOffset(Camera Offset Amount=5)
  ]
  rendererClasses [&NiagaraSpriteRendererProperties]
  renderers [{rendererClass=&NiagaraSpriteRendererProperties}]
 }
 MainBurst {
  emitterSpawnScript [
   &InitializeAudioPlayer(Audio Player=DataInterface() {propertyValues {
      Attenuation=&ATT_BasilOkra_Meltdown_Steam
      Concurrency=null
      MaxPlaysPerTick=10
      SoundToPlay=/WildEstate_BasilOkra/Assets/Audio/SFX/MSS_BasilOkra_Meltdown_SteamBurst_Long
      bAllowLoopingOneShotSounds=false
      bLimitPlaysPerTick=true
      bOnlyActiveDuringGameplay=false
      bStopWhenComponentIsDestroyed=true
      ConfigurationUserParameter {parameter {name=None, typeDefHandle {registeredTypeIndex=1}, varData []}}
      ParameterNames []
     }})
   /SetVariables(
    Emitter.Length=122
    Emitter.ParticlesToSpawn=12
    Emitter.Random:NiagaraFloat=&RandomRangeFloat(Minimum=0, Maximum=1)
   ) {isSetParameters=true, module=SetVariables_#5}
   /Niagara/Modules/Audio/PlayAudio()
  ]
  emitterUpdateScript [
   &EmitterState(Life Cycle Mode=::NewEnumerator1, Loop Behavior=::NewEnumerator1, Loop Duration Mode=::NewEnumerator1, UseLoopDelay=false)
   &SpawnBurst_Instantaneous(Spawn Count=$Emitter.ParticlesToSpawn, Use Spawn Probability=false, UseLoopCountLimit=false)
   /SetVariables(Emitter.Reveal:NiagaraFloat=&Multiply_Float(
     A=&Lerp_Float(
      A=&FloatFromCurve(FloatCurve=DataInterface() {propertyValues {Curve {keys [<at 0 0>,<at 0.6 1>]}}}, CurveIndex=$User.Reveal)
      B=&FloatFromCurve(FloatCurve=DataInterface() {propertyValues {Curve {keys [<at 0 0>,<at 0.95 1>]}}}, CurveIndex=$User.Reveal)
      Alpha=$System.SystemRandom
     )
     B=&Lerp_Float(A=1, B=1.1, Alpha=$Emitter.Random)
    )) {isSetParameters=true, module=SetVariables_#6}
   /SetVariables(Emitter.Length:NiagaraFloat=&Multiply_Float(
     A=$Emitter.Initial.Length
     B=&FloatFromCurve(
      FloatCurve=DataInterface() {propertyValues {Curve {keys [
          {interpMode=RCIM_Cubic, leaveTangent=6.4155, tangentMode=RCTM_Break, time=0, value=0}
          {arriveTangent=0.1191, interpMode=RCIM_Cubic, leaveTangent=0.1191, tangentMode=RCTM_User, time=0.1, value=0.9}
          {arriveTangent=0.1177, interpMode=RCIM_Cubic, leaveTangent=0.1177, tangentMode=RCTM_User, time=1, value=1}
         ]}}}
      CurveIndex=$Emitter.Reveal
      Scale Curve=&Multiply_Float(A=$User.FXScale, B=&Lerp_Float(A=0.7, B=1, Alpha=$System.SizeRandom))
     )
    )) {isSetParameters=true, module=SetVariables_#7}
  ]
  particleSpawnScript [
   &InitializeParticle(Color Mode=::NewEnumerator0, Sprite Rotation Mode=::NewEnumerator2, UsePositionOffset=false, Write Material Random=true)
   &DynamicMaterialParameters(
    Index 0 Param 1=$Emitter.Random
    Module.Index0Param0Write=true
    Module.Index0Param1Write=true
    Module.Index0Param2Write=true
    Module.Index0Param3Write=true
   )
  ]
  particleUpdateScript [
   &ParticleState(Kill Particles When Lifetime Has Elapsed=false, Let Infinitely Lived Particles Die When Emitter Deactivates=true)
   /SetVariables(Particles.ParticleRelativeLifetimeRibbons:NiagaraFloat=&ReturnNormalizedExecIndex()) {isSetParameters=true, module=SetVariables_#8}
   /SetVariables(Particles.OpacityMultiplierRibbons:NiagaraFloat=&FloatFromCurve(
     FloatCurve=DataInterface() {propertyValues {Curve {keys [
         {interpMode=RCIM_Cubic, tangentMode=RCTM_User, time=0, value=0}
         {arriveTangent=0.007, interpMode=RCIM_Cubic, leaveTangent=0.007, tangentMode=RCTM_User, time=0.5, value=1.5}
         {arriveTangent=-0.7513, interpMode=RCIM_Cubic, leaveTangent=-0.7513, tangentMode=RCTM_User, time=0.9, value=1}
         {arriveTangent=-0.9808, interpMode=RCIM_Cubic, leaveTangent=-0.9808, tangentMode=RCTM_User, time=1.7, value=0}
        ]}}}
     CurveIndex=&Add_Float(
      A=&Multiply_Float(A=$Particles.ParticleRelativeLifetimeRibbons, B=-1)
      B=&Multiply_Float(A=$Emitter.Reveal, B=2)
     )
    )) {isSetParameters=true, module=SetVariables_#9}
   /SetVariables(Particles.InterpolateCurveRibbons:NiagaraFloat=&FloatFromCurve(
     FloatCurve=DataInterface() {propertyValues {Curve {keys [{arriveTangent=-0.6503, time=0, value=0}, {arriveTangent=0.8679, time=1, value=1}]}}}
     CurveIndex=$Particles.ParticleRelativeLifetimeRibbons
    )) {isSetParameters=true, module=SetVariables_#10}
   /SetVariables(Particles.PositionRibbons:NiagaraPosition=&AddVectorToPosition(
     Position=&AddVectorToPosition(
      Position=&SimulationPosition()
      Vector=&TransformVector(Vector=<0 0 0>, Destination Space=::Simulation)
     )
     Vector=&Multiply_VectorByFloat(
      Vector=&RotateVectorByQuaternion(
       Vector To Rotate=<0 1 0>
       Quaternion=&LerpQuaternion(
        Quaternion A:Quat4f=$Engine.Owner.Rotation
        Quaternion B:Quat4f=$System.InterpolateOverTime.InterpolateOverTime_Quaternion.Moving Average
        Lerp Factor:NiagaraFloat=$Particles.InterpolateCurveRibbons
        Enable Spherical Linear Interpolation:NiagaraBool=false
       )
      )
      Float=&Multiply_Float(A=$Emitter.Length, B=$Particles.InterpolateCurveRibbons)
     )
    )) {isSetParameters=true, module=SetVariables_#11}
   /SetVariables(
    Particles.RibbonColor:LinearColor=&MultiplyLinearColor(
     Linear Color=&MultiplyLinearColor(
      Linear Color=&ColorFromCurve(
       ColorCurve=DataInterface() {propertyValues {
         AlphaCurve {keys [
           <at 0 0>
           <at 0.0346 1>
           <at 0.2133 0.8>
           {arriveTangent=-1.0175, interpMode=RCIM_Cubic, leaveTangent=-1.0175, time=0.7268, value=0.8}
           <at 0.9995 0 RCIM_Cubic>
          ]}
         BlueCurve {keys [<at 0.4438 1>]}
         GreenCurve {keys [<at 0.4438 1>]}
         RedCurve {keys [<at 0.4438 1>]}
        }}
       CurveIndex=$Particles.ParticleRelativeLifetimeRibbons
      )
      Scale Factor=&MakeVector4_FromLinearColor(LinearColor=&FNTimeOfDayColor(Night Color=<rgba 1 1 1 1>, Morning Color=<rgba 0.8 0.9067 1 1>, Day Color=<rgba 1 1 1 1>, Evening Color=<rgba 1 0.98 0.983 1>))
     )
     Scale Factor=&MakeFloat4FromVectorAndFloat(
      "Vector (RGB)"=&VectorFromFloat(Value=&FloatFromCurve(FloatCurve=DataInterface() {propertyValues {Curve {keys [<at 0.45 1 RCIM_Cubic>]}}}, CurveIndex=$User.Reveal))
      "Float (Alpha)"=1
     )
    )
    Particles.RibbonWidth=&FloatFromCurve(
     FloatCurve=DataInterface() {propertyValues {Curve {keys [
         {arriveTangent=0.675, interpMode=RCIM_Cubic, leaveTangent=0.675, tangentMode=RCTM_User, time=0, value=0.2}
         {arriveTangent=1.0523, interpMode=RCIM_Cubic, leaveTangent=1.0523, tangentMode=RCTM_User, time=0.2799, value=0.5408}
         {arriveTangent=-0.0372, interpMode=RCIM_Cubic, leaveTangent=-0.0372, tangentMode=RCTM_User, time=0.7, value=0.8}
         {arriveTangent=-1.5957, interpMode=RCIM_Cubic, leaveTangent=-1.5957, tangentMode=RCTM_User, time=1, value=0.5138}
        ]}}}
     CurveIndex=$Particles.InterpolateCurveRibbons
     Scale Curve=&Multiply_Float(A=25, B=$User.FXScale)
    )
   ) {isSetParameters=true, module=SetVariables_#12}
   /SetVariables(Particles.ScaledRibbonWidth:NiagaraFloat=&Multiply_Float(
     A=$Particles.RibbonWidth
     B=&Multiply_Float(
      A=&FloatFromCurve(
       FloatCurve=DataInterface() {propertyValues {Curve {keys [
           {interpMode=RCIM_Cubic, leaveTangent=0.5491, tangentMode=RCTM_Break, time=0, value=0.7}
           {arriveTangent=1.0086, interpMode=RCIM_Cubic, leaveTangent=1.0086, tangentMode=RCTM_User, time=0.3931, value=1}
           {arriveTangent=0.0216, interpMode=RCIM_Cubic, tangentMode=RCTM_Break, time=1, value=1.5}
          ]}}}
       CurveIndex=$Emitter.Reveal
       Scale Curve=&Lerp_Float(
        A=1.6
        B=&FloatFromCurve(
         FloatCurve=DataInterface() {propertyValues {Curve {keys [<at 0 1.9 RCIM_Constant>,<at 0.333 2.25 RCIM_Constant>,<at 0.666 2.6 RCIM_Constant>]}}}
         CurveIndex=$Emitter.Random
        )
        Alpha=$System.SizeRandom
       )
      )
      B=1
     )
    )) {isSetParameters=true}
   /SetVariables(Particles.ScaledRibbonColor:LinearColor=&MultiplyLinearColor(
     Linear Color=$Particles.RibbonColor
     Scale Factor=&MakeFloat4FromVectorAndFloat(
      "Vector (RGB)"=&MakeCustomVectorFromBool(Bool=$User.Is_Purple, True Vector=<0.7 0 1>, False Vector=<1 1 1>)
      "Float (Alpha)"=&Multiply_Float(
       A=&FloatFromCurve(
        FloatCurve=DataInterface() {propertyValues {Curve {
           defaultValue=1
           keys [{interpMode=RCIM_Cubic, tangentMode=RCTM_User, time=0, value=0}, {interpMode=RCIM_Cubic, tangentMode=RCTM_User, time=1, value=1}]
          }}}
        CurveIndex=$Particles.Age
       )
       B=&Multiply_Float(
        A=$Particles.OpacityMultiplierRibbons
        B=&Multiply_Float(
         A=&FloatFromCurve(
          FloatCurve=DataInterface() {propertyValues {Curve {keys [
              {interpMode=RCIM_Cubic, tangentMode=RCTM_Break, time=0.6, value=1}
              {arriveTangent=-4.2498, interpMode=RCIM_Cubic, tangentMode=RCTM_Break, time=0.99, value=0}
             ]}}}
          CurveIndex=$Emitter.Reveal
          Scale Curve=&Lerp_Float(A=0.8, B=1, Alpha=$Emitter.Random)
         )
         B=$System.OpacityRandom
        )
       )
      )
     )
    )) {isSetParameters=true}
   &CameraOffset(Camera Offset Amount=5)
   &ScaleColor(
    Scale Alpha=&FloatFromCurve(
     FloatCurve=DataInterface() {propertyValues {Curve {
        defaultValue=1
        keys [{interpMode=RCIM_Cubic, tangentMode=RCTM_User, time=0, value=0}, {interpMode=RCIM_Cubic, tangentMode=RCTM_User, time=1, value=1}]
       }}}
     CurveIndex=$Emitter.Age
    )
    ScaleA=true
    ScaleRGB=true
   ) {enabled=false}
  ]
  rendererClasses [&NiagaraRibbonRendererProperties, &NiagaraSpriteRendererProperties]
  renderers [
   {rendererClass=&NiagaraRibbonRendererProperties}
   {rendererClass=&NiagaraSpriteRendererProperties, rendererIndex=1}
  ]
 }
}
summary {
 systemName=NS_BasilOkra_Pipe_Steam_Small
 emitters [
  {emitterName=InitialFlameBurst, rendererClasses [&NiagaraSpriteRendererProperties]}
  {emitterName=AdditionalFlames, rendererClasses [&NiagaraSpriteRendererProperties]}
  {emitterName=MainBurst, rendererClasses [&NiagaraRibbonRendererProperties, &NiagaraSpriteRendererProperties]}
 ]
 userVariables [
  User.bIsFrontEnd {type=NiagaraBool}
  User.bIsFrontEndPreview {type=NiagaraBool}
  User.bIsLargeBodyPawn {type=NiagaraBool}
  User.FXScale {type=NiagaraFloat, defaultValue=3, description=""}
  User.Is_Purple {type=NiagaraBool}
  User.Reveal {type=NiagaraFloat, defaultValue=0.5, description=""}
 ]
}
